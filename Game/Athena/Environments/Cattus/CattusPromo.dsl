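; Cattus promo hologram projector actor that is revealed in a match when a named calendar event becomes active.
; BeginPlay waits for match ready on authority, watches that event, then sets Visible, whose OnRep unhides the actor.
; The construction script picks one of four poster materials or an override, scales the mesh, and creates screen and projection dynamic materials.
; Exposed variables drive their rotation, bend, color, glow, contrast and optional texture.
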(event EventBeginPlay
  (Class|FortAsyncActionAuthorityMatchReady|MatchReadyAsync
    (:OnReady
      (MCPContext|WatchCalendarEvent (Variables|Default|GetCalenderEventName)
        (:OnEventActive
          (Variables|Default|SetVisible true))))))

(fn OnRep_Visible
  (Rendering|SetActorHiddenInGame false))

(fn ConstructionScript
  (bind _m_cattuspromo4 (Variables|Default|GetM_CattusPromo4))
  (bind _m_cattuspromo3 (Variables|Default|GetM_CattusPromo3))
  (bind _m_cattuspromo2 (Variables|Default|GetM_CattusPromo2))
  (bind _m_cattuspromo1 (Variables|Default|GetM_CattusPromo1))
  (bind _array (Utilities|Array|MakeArray _m_cattuspromo1 _m_cattuspromo2 _m_cattuspromo3 _m_cattuspromo4))
  (bind _sm_holoprojector (Variables|Default|GetSM_HoloProjector))
  (bind _mid_projection (Variables|Default|GetMID_Projection))
  (Rendering|Material|SetMaterial (Variables|Default|GetSM_HoloProjector) 0 (Utilities|Array|Get(acopy) _array (Variables|Default|GetPosterSelect)))
  (Utilities|Array|Clear _array)
  (Transformation|SetWorldScale3D (Variables|Default|GetSM_HoloProjector) (* (Variables|Default|GetScale) (Variables|Default|GetScaleUniform)))
  (if (Utilities|IsValid (Variables|Default|GetOverrideMaterial_Screen) :!class /Script/Engine.KismetSystemLibrary)
    (Rendering|Material|SetMaterial (Variables|Default|GetSM_HoloProjector) 0 (Variables|Default|GetOverrideMaterial_Screen)))
  (bind _returnvalue (Rendering|Material|CreateDynamicMaterialInstance _sm_holoprojector 1 0 "None" :!class /Script/Engine.PrimitiveComponent))
  (Variables|Default|SetMID_Projection _returnvalue)
  (bind _returnvalue_2 (Rendering|Material|CreateDynamicMaterialInstance _sm_holoprojector 0 0 "None" :!class /Script/Engine.PrimitiveComponent))
  (Variables|Default|SetMID_Screen _returnvalue_2)
  (Rendering|Material|SetScalarParameterValue (Variables|Default|GetMID_Projection) "Constant Rotation Speed" (Variables|Default|GetRotationSpeed) :!class /Script/Engine.MaterialInstanceDynamic)
  (Rendering|Material|SetScalarParameterValue _mid_projection "Rotationamount" (* (Variables|Default|GetBend) -0.01) :!class /Script/Engine.MaterialInstanceDynamic)
  (Rendering|Material|SetVectorParameterValue _mid_projection "Color" (Variables|Default|GetColor) :!class /Script/Engine.MaterialInstanceDynamic)
  (Rendering|Material|SetScalarParameterValue (Variables|Default|GetMID_Screen) "Glow" (Variables|Default|GetGlowScreen) :!class /Script/Engine.MaterialInstanceDynamic)
  (Rendering|Material|SetScalarParameterValue (Variables|Default|GetMID_Projection) "Glow" (Variables|Default|GetGlowLaser) :!class /Script/Engine.MaterialInstanceDynamic)
  (Rendering|Material|SetScalarParameterValue (Variables|Default|GetMID_Screen) "Contrast" (Variables|Default|GetScreenContrast) :!class /Script/Engine.MaterialInstanceDynamic)
  (Utilities|IsValid (Variables|Default|GetTexture)
    (:Is%20Valid
      (Rendering|Material|SetTextureParameterValue (Variables|Default|GetMID_Projection) "BaseColor" (Variables|Default|GetTexture)))))
