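# Nanite-enabled environment master material for Cattus skull and bone props, with static switches for sword textures and teeth.
# World-aligned, Fresnel-tinted bone or sword textures receive baked normals and mask occlusion, then height-lerped dust, before normals convert to world space.
# The teeth switch adds hit-effect World Position Offset.
# Runtime virtual textures blend landscape color and normals by height above terrain; mobile skips this blend and uses only the Creative move-tool output.
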Material /Game/Athena/Environments/Cattus/Materials/M_Cattus_Env
graph {
 properties {
  bTangentSpaceNormal=false
  bUseMaterialAttributes=true
  bUsedWithNanite=true
  pixelDepthOffsetMode=PDOM_Legacy
  usedShadingModels=Default Lit
 }
 nodes [
  n1 = TextureSampleParameter2D() {
   parameterName=Basecolor
   samplerSource=SSM_Wrap_WorldGroupSettings
   texture=/Game/Athena/Prototype/Blueprints/Doggus_Sword/Textures/T_Doggus_Sword_D
  }
  n2 = TextureSample() {samplerSource=SSM_Wrap_WorldGroupSettings, texture=/Game/Athena/Prototype/Blueprints/Doggus_Sword/Textures/T_Doggus_Sword_E}
  n3 = TextureSample() {
   samplerSource=SSM_Wrap_WorldGroupSettings
   samplerType=SAMPLERTYPE_Masks
   texture=/Game/Athena/Prototype/Blueprints/Doggus_Sword/Textures/T_Doggus_Sword_FX
  }
  n4 = TextureSampleParameter2D() {
   parameterName=Specular
   samplerSource=SSM_Wrap_WorldGroupSettings
   samplerType=SAMPLERTYPE_Masks
   texture=/Game/Athena/Prototype/Blueprints/Doggus_Sword/Textures/T_Doggus_Sword_S
  }
  n5 = SetMaterialAttributes!(
   MaterialAttributes
   Base Color=StaticSwitchParameter(True=LinearInterpolate(A=@n1.RGB, B=@n2.RGB, Alpha=@n3.R), False=@n1.RGB) {defaultValue=true, parameterName=UseEmissiveBlend}
   Metallic=@n4.G
   Specular=@n4.G
   Roughness=@n4.G
  ) {attributeSetTypes [
    #1
    #2
    #3
    #4
   ]}
  MaterialExpressionReroute_1 = Reroute(None=@n5)
  n6 = TextureSampleParameter2D() {
   parameterName=Mask
   samplerSource=SSM_Wrap_WorldGroupSettings
   samplerType=SAMPLERTYPE_Masks
   texture=/Game/Athena/Environments/Cattus/Demo/Textures/Cattus_Skull_Outside_M
  }
  n7 = LinearInterpolate(
   A=LinearInterpolate(
    A=VectorParameter() {group=Bone, parameterName=Color1, defaultValue=<rgba 0.3698 0.2375 0.089 1>}.RGB
    B=VectorParameter() {group=Bone, parameterName=Color2, defaultValue=<rgba 0.5747 0.5747 0.5747 1>}.RGB
    Alpha=@n6.G
   )
   B=VectorParameter() {group=Bone, parameterName=Color3, defaultValue=<rgba 0.1892 0.1685 0.1277 1>}.RGB
   Alpha=@n6.B
  )
  n8 = Fresnel(ExponentIn=ScalarParameter() {defaultValue=1.3, group=Bone, parameterName=FresnelBone}, Normal=VertexNormalWS()) {baseReflectFraction=0}
  n9 = TextureObject() {texture=/Game/Athena/Environments/Cattus/Textures/T_Bone_D}
  n10 = TextureObject() {samplerType=SAMPLERTYPE_Normal, texture=/Game/Athena/Environments/Cattus/Textures/T_Bone_N}
  n11 = MaterialFunctionCall!(
   Diffuse Texture Object=@n9
   Spec Texture Object=TextureObject() {samplerType=SAMPLERTYPE_Masks, texture=/Game/Athena/Environments/Cattus/Textures/T_Bone_M}
   TextureSize=ScalarParameter() {defaultValue=369.8548, group=Bone, parameterName=SizeBone_Tiling}
   --------------------------
   Use seperate settings for Z projection
   Use seperate texture for Z projection
   Texture Size Z projection
   Z Plane Diffuse Texture Object
   Z Plane Spec Texture Object
   "--------------------------      "
   ProjectionTransitionContrast=ScalarParameter() {defaultValue=2, group=Bone, parameterName=BoneProjectionContrast}
   Use Texture Alpha Heightmap to Refine the Transistion
   HeightLerpTransitionContrast
   PickTheDiffuseHeightChannel
   TheDiffuseHeightChannel
   "--------------------------       "
   NormalTextureObject=@n10
   World Space Normals=StaticBool()
   "--------------------------  "
   Export Float 4 Diffuse
   Export Float 4 Spec
   "--------------------------     "
   WorldPosition
   World Space Normal
  ) {materialFunction=/Engine/Functions/Engine_MaterialFunctions02/Texturing/WorldAlignedTextures_Complex}
  n12 = MaterialFunctionCall(Float3=@n11.XYZ Spec) {materialFunction=&BreakOutFloat3Components}
  n13 = Add(A=@n7, B=Multiply(A=Multiply(A=@n7, B=@n8), B=@n12.B))
  n14 = Multiply(A=TextureCoordinate(), B=ScalarParameter() {defaultValue=55, group=Bone, parameterName=Details UV Tile})
  n15 = TextureSample(UVs=@n14, Tex=@n9)
  n16 = NamedRerouteDeclaration(None=@n11.XYZ Diffuse Texture) {name=BoneD, nodeColor=<rgba 0 0.5781 1 1>}
  n17 = ComponentMask(None=VertexNormalWS()) {b=true}
  n18 = ScalarParameter() {defaultValue=0.4, parameterName=Bone Spec}
  n19 = Constant3Vector() {constant=<rgba 1 1 0 0>}
  n20 = SetMaterialAttributes!(
   MaterialAttributes
   Base Color=Multiply(A=ShadingPathSwitch(Default=@n13, Mobile=@n7), B=Multiply(A=@n15.R, B=@n16))
   Specular=Multiply(A=Max(A=@n17) {constB=0.25}, B=Multiply(A=@n18, B=@n12.R))
   Roughness=LinearInterpolate(
    A=ScalarParameter() {defaultValue=0.1, group=Bone, parameterName=RoughnessBoneMin}
    B=ScalarParameter() {defaultValue=1.2, group=Bone, parameterName=RoughnessBoneMax}
    Alpha=@n12.B
   )
   Normal=Add(A=Multiply(A=@n11.Normal XYZ Texture, B=@n19), B=TextureSample(UVs=@n14, Tex=@n10).RGB)
  ) {attributeSetTypes [
    #1
    #3
    #4
    #5
   ]}
  MaterialExpressionReroute_3 = Reroute(None=@n20)
  n21 = StaticSwitchParameter(True=@MaterialExpressionReroute_1.Sword Textures, False=@MaterialExpressionReroute_3.Bone Surf) {parameterName=IsSword}
  Composite!(None=@n21) {subgraphName=AO+N baked}
  Composite!()
  Composite() {subgraphName="Sword ( Just in case)"}
  Composite() {subgraphName=Bones}
  MaterialExpressionReroute_8 = Reroute(None=@n21)
  n22 = GetMaterialAttributes(None=@MaterialExpressionReroute_8.Input) {attributeGetTypes [#5, #1]}
  n23 = Constant3Vector() {constant=<rgba 1 1 0 0>}
  n24 = TextureSampleParameter2D() {parameterName=Baked Normal, samplerType=SAMPLERTYPE_Normal, texture=/Game/Athena/Prototype/Blueprints/Doggus_Sword/Textures/T_Doggus_Sword_N}
  n25 = TextureSampleParameter2D() {
   parameterName=Mask
   samplerSource=SSM_Wrap_WorldGroupSettings
   samplerType=SAMPLERTYPE_Masks
   texture=&Cattus_Skull_Outside_M
  }
  n26 = SetMaterialAttributes!(
   MaterialAttributes=@n22.MaterialAttributes
   Normal=Add(A=Multiply(A=@n22.Normal, B=@n23), B=@n24.RGB)
   Base Color=Multiply(A=@n22.Base Color, B=@n25.R)
  ) {attributeSetTypes [#5, #1]}
  MaterialExpressionReroute_7 = Reroute(None=@n26)
  MaterialExpressionReroute_5 = Reroute(None=@MaterialExpressionReroute_7.AO + N)
  n27 = GetMaterialAttributes(None=@MaterialExpressionReroute_5.Input) {attributeGetTypes [#1]}
  n28 = Multiply(A=@n27.Base Color, B=VectorParameter() {group=Dust, parameterName=Dust, defaultValue=<rgba 0.7726 0.5853 0.4282 0>}.RGB)
  n29 = Constant() {r=0.75}
  n30 = Divide(A=NamedRerouteUsage(), B=ScalarParameter() {defaultValue=2000, group=Dust, parameterName=Dust Distance})
  n31 = TextureSampleParameter2D() {
   parameterName=Mask
   samplerSource=SSM_Wrap_WorldGroupSettings
   samplerType=SAMPLERTYPE_Masks
   texture=&Cattus_Skull_Outside_M
  }
  n32 = Multiply(A=@n31.R, B=Constant() {r=1.5})
  n33 = MaterialFunctionCall(
   NoRVT=@MaterialExpressionReroute_5.Input
   RVT=BlendMaterialAttributes(
    A=SetMaterialAttributes!(MaterialAttributes=@n27.MaterialAttributes, Base Color=@n28, Roughness=@n29) {attributeSetTypes [#1, #4]}
    B=@MaterialExpressionReroute_5.Input
    Alpha=MaterialFunctionCall(Transition Phase=@n30, Height Texture=@n32) {materialFunction=&HeightLerp}.Lerp Alpha No Contrast
   )
  ) {materialFunction=&FN_RVT_Switch_Material}
  MaterialExpressionReroute_2 = Reroute(None=@n33.Result)
  Composite!(None=@MaterialExpressionReroute_2.DustLayer) {subgraphName=Tangent to World}
  MaterialExpressionReroute_9 = Reroute(None=@MaterialExpressionReroute_2.DustLayer)
  n34 = GetMaterialAttributes(None=@MaterialExpressionReroute_9.Input) {attributeGetTypes [#5]}
  n35 = SetMaterialAttributes!(MaterialAttributes=@n34.MaterialAttributes, Normal=Transform(None=@n34.Normal)) {attributeSetTypes [#5]}
  MaterialExpressionReroute_10 = Reroute(None=@n35)
  Composite!(MaterialAttributes=@MaterialExpressionReroute_10.N Tangent To World) {subgraphName=WPO Bounce}
  Composite!(None=@MaterialExpressionReroute_7.AO + N) {subgraphName=Dust}
  MaterialExpressionReroute_11 = Reroute(None=@MaterialExpressionReroute_10.N Tangent To World)
  n36 = MaterialFunctionCall!(
   WSDeformationCenter=VectorParameter() {bUseCustomPrimitiveData=true, group=z_code, parameterName=WSDeformationCenter, primitiveDataIndex=20}.RGB
   Code_DestructionAmount=ScalarParameter() {group=z_code, parameterName=Code_DestructionAmount}
   Code_UseUvForDamage
   Code_UseHitEffects=StaticBoolParameter() {defaultValue=true, group=z_code, parameterName=Code_UseHitEffects}
   Code_Emissive
   WSDamageRadius=ScalarParameter() {bUseCustomPrimitiveData=true, defaultValue=1, group=z_code, parameterName=WSDamageRadius, primitiveDataIndex=26}
   CrackDamageShader
   Diffuse
   ScalarModDamageCracks
   ScalarOffsetforDamageCracks
   ExposeLerpFunctionForInvertingTheHitDistanceCalc
   InvertDamageFalloff
   CrackWorldPosition
   WorldPosition
   DistanceCullFadeAnimationOffset=ScalarParameter() {bUseCustomPrimitiveData=true, group=z_code, parameterName=DistanceCullFadeAnimationOffset, primitiveDataIndex=12}
  ) {materialFunction=/Game/Packages/jl_tests/Fortnite_HitEffects}
  MaterialExpressionReroute_0 = Reroute(None=@n36.Hit And Shrink WPO)
  n37 = SetMaterialAttributes!(MaterialAttributes=@MaterialExpressionReroute_11.MaterialAttributes, World Position Offset=@MaterialExpressionReroute_0.Hit And Shrink WPO) {attributeSetTypes [#6]}
  MaterialExpressionReroute_12 = Reroute(None=@n37)
  n38 = StaticSwitchParameter(True=@MaterialExpressionReroute_12.WPO Wobble, False=@MaterialExpressionReroute_10.N Tangent To World) {parameterName=IsTeeth}
  Composite!(InputPin=@n38) {subgraphName=RVT}
  NamedRerouteUsage()
  PinBase() {reroutePins [Input {expression=&self:MaterialExpressionReroute_8}]}
  PinBase!(AO + N=@n26) {reroutePins [AO + N {expression=&self:MaterialExpressionReroute_7}]}
  PinBase!() {reroutePins [MaterialAttributes {expression=&self:MaterialExpressionReroute_11}]}
  PinBase!(WPO Wobble=@n37) {reroutePins [WPO Wobble {expression=&self:MaterialExpressionReroute_12}]}
  PinBase!() {reroutePins [Input {expression=&self:MaterialExpressionReroute_5}]}
  PinBase!(DustLayer=@n33.Result) {reroutePins [DustLayer {expression=&self:MaterialExpressionReroute_2}]}
  PinBase!() {reroutePins [InputPin {expression=&self:MaterialExpressionReroute_24}]}
  n39 = RuntimeVirtualTextureSampleParameter() {materialType=BaseColor_Normal_Specular_YCoCg, parameterName=RVT, virtualTexture=/Game/Athena/Apollo/Maps/Landscape/VT/RVT_Apollo_Landscape}
  MaterialExpressionReroute_24 = Reroute(None=@n38)
  n40 = RuntimeVirtualTextureSampleParameter() {materialType=WorldHeight, parameterName=RVTHeight, virtualTexture=/Game/Athena/Apollo/Maps/Landscape/VT/RVT_Apollo_Landscape_Height}
  n41 = Subtract(A=ComponentMask(None=WorldPosition().XYZ) {b=true}, B=@n40.WorldHeight)
  n42 = NamedRerouteDeclaration(None=NamedRerouteDeclaration(None=@n41) {name=RVT Raw, nodeColor=<rgba 1 0 0.4219 1>}) {name=RVT Height, nodeColor=<rgba 0.2422 1 0 1>}
  n43 = ScalarParameter() {defaultValue=8, group=RVT, parameterName=Blend RVT Normals}
  n44 = BlendMaterialAttributes(
   A=SetMaterialAttributes!(MaterialAttributes=@MaterialExpressionReroute_24.InputPin, Normal=@n39.Normal) {attributeSetTypes [#5]}
   B=@MaterialExpressionReroute_24.InputPin
   Alpha=Saturate(None=Divide(A=@n42, B=@n43))
  ) {vertexAttributeBlendType=UseA}
  n45 = SetMaterialAttributes!(MaterialAttributes=@n44, Base Color=@n39.BaseColor, Specular=@n39.BaseColor, Roughness=@n39.BaseColor, Emissive Color=Constant()) {attributeSetTypes [
    #1
    #3
    #4
    #7
   ]}
  n46 = ScalarParameter() {defaultValue=100, group=RVT, parameterName=Blend RVT}
  n47 = BlendMaterialAttributes(A=@n45, B=@n44, Alpha=Saturate(None=Divide(A=@n41, B=@n46))) {vertexAttributeBlendType=UseA}
  PinBase!(Output=@n47) {reroutePins [Output {expression=&self:MaterialExpressionReroute_23}]}
  PinBase!()
  PinBase!(Hit And Shrink WPO=@n36.Hit And Shrink WPO) {reroutePins [Hit And Shrink WPO {expression=&self:MaterialExpressionReroute_0}]}
  PinBase!()
  PinBase!(Sword Textures=@n5) {reroutePins [Sword Textures {expression=&self:MaterialExpressionReroute_1}]}
  PinBase!()
  PinBase!(Bone Surf=@n20) {reroutePins [Bone Surf {expression=&self:MaterialExpressionReroute_3}]}
  PinBase!() {reroutePins [Input {expression=&self:MaterialExpressionReroute_9}]}
  PinBase!(N Tangent To World=@n35) {reroutePins [N Tangent To World {expression=&self:MaterialExpressionReroute_10}]}
  Reroute(None=@n11.XYZ Diffuse Texture)
  n48 = MaterialFunctionCall(MA In=@n38) {materialFunction=&MF_CreativeMoveTool}
  MaterialExpressionReroute_23 = Reroute(None=@n47)
  ShadingPathSwitch(
   Default=MaterialFunctionCall(NoRVT=@n48.MA Out, RVT=MaterialProxyReplace(Realtime=@MaterialExpressionReroute_23.Output, MaterialProxy=@n38)) {materialFunction=&FN_RVT_Switch_Material}.Result
   Mobile=@n48.MA Out
  )
 ]
}
