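# Switchable master material for Cattus environment pieces: tinted world-aligned bone, gradient-tinted teeth, or a height-blended sword with emissive-texture cracks in base color.
# Optional moss layering height-blends world-aligned gravel and grass using vertex colors and normals, with optional purple growth.
# Layered normals are rebuilt into tangent space and added to the mesh normal map.
# Mobile/forward paths override specular and roughness, skipping optional AO; teeth add hit-effect vertex offset.
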Material /Game/Athena/Environments/Cattus/Materials/M_Cattus_Env_CP
graph {
 properties {bUseMaterialAttributes=true, pixelDepthOffsetMode=PDOM_Legacy, usedShadingModels=Default Lit}
 nodes [
  n1 = VectorParameter() {parameterName=ColorSwordBlend, defaultValue=<rgba 0.2604 0.2048 0.1397 1>}
  n2 = TextureSampleParameter2D() {
   parameterName=Basecolor
   samplerSource=SSM_Wrap_WorldGroupSettings
   texture=/Game/Athena/Prototype/Blueprints/Doggus_Sword/Textures/T_Doggus_Sword_D
  }
  n3 = TextureSample() {samplerSource=SSM_Wrap_WorldGroupSettings, texture=/Game/Athena/Prototype/Blueprints/Doggus_Sword/Textures/T_Doggus_Sword_E}
  n4 = BumpOffset(
   Coordinate=Multiply(A=TextureCoordinate(), B=ScalarParameter() {defaultValue=5, parameterName=TilingCracks})
   Height=Constant() {r=1}
   HeightRatioInput=LinearInterpolate(
    A=ScalarParameter() {defaultValue=-0.0476, parameterName=BumpOffsetCracks1}
    B=ScalarParameter() {defaultValue=-0.1524, parameterName=BumpOffsetCracks2}
    Alpha=MaterialFunctionCall() {materialFunction=&DitherTemporalAA}.Result
   )
  )
  n5 = Add(
   A=@n3.RGB
   B=Multiply(
    A=VectorParameter() {parameterName=CrackColor, defaultValue=<rgba 0.1927 0.0825 0.187 1>}.RGB
    B=TextureSample(UVs=@n4) {samplerSource=SSM_Wrap_WorldGroupSettings, texture=/Game/Effects/Fort_Effects/Textures/T_Cracks_Gradient}.R
   )
  )
  n6 = TextureSample() {
   samplerSource=SSM_Wrap_WorldGroupSettings
   samplerType=SAMPLERTYPE_Masks
   texture=/Game/Athena/Prototype/Blueprints/Doggus_Sword/Textures/T_Doggus_Sword_FX
  }
  n7 = StaticSwitchParameter(
   True=LinearInterpolate(A=@n2.RGB, B=ShadingPathSwitch(Default=@n5, Deferred=@n5, Forward=@n3.RGB, Mobile=@n3.RGB), Alpha=@n6.R)
   False=@n2.RGB
  ) {defaultValue=true, parameterName=UseEmissiveBlend}
  n8 = CollectionParameter() {collection=&/MPC_CattusPOI, parameterName=SwordHeightBlend_Offset}
  n9 = Subtract(A=ComponentMask(None=WorldPosition().XYZ) {b=true}, B=MaterialProxyReplace(Realtime=@n8, MaterialProxy=Subtract(A=@n8) {constB=768}))
  n10 = CollectionParameter() {collection=&MPC_CattusPOI, parameterName=SwordHeightBlend_Distance}
  n11 = Saturate(None=Add(A=Divide(A=@n9, B=@n10), B=ScalarParameter() {defaultValue=0.1905, parameterName=SwordBlendOpacity}))
  n12 = TextureSampleParameter2D() {
   parameterName=Specular
   samplerSource=SSM_Wrap_WorldGroupSettings
   samplerType=SAMPLERTYPE_Masks
   texture=/Game/Athena/Prototype/Blueprints/Doggus_Sword/Textures/T_Doggus_Sword_S
  }
  n13 = TextureSample(UVs=Multiply(A=TextureCoordinate(), B=ScalarParameter() {defaultValue=7.1097, parameterName=TilingDetail})) {samplerSource=SSM_Wrap_WorldGroupSettings, samplerType=SAMPLERTYPE_Masks, texture=/Game/Global/Textures/Default/Noise/T_Metal_Noise}
  n14 = Multiply(A=@n13.R, B=ScalarParameter() {defaultValue=0.13, parameterName=RoughnessNoise})
  n15 = VectorParameter() {parameterName=TeethGradientColor, defaultValue=<rgba 0.2344 0.2219 0.1807 1>}
  n16 = VectorParameter() {parameterName=GradientDirection, defaultValue=<rgba 0 0 1 1>}
  n17 = DotProduct(
   A=Normalize(VectorInput=Transform(None=@n16.RGB) {transformSourceType=TRANSFORMSOURCE_Local})
   B=Subtract(A=WorldPosition().XYZ, B=ObjectPositionWS())
  )
  n18 = Multiply(A=ObjectRadius(), B=ScalarParameter() {defaultValue=2, parameterName=Gradient})
  n19 = Clamp(None=Add(A=Divide(A=@n17, B=@n18), B=ScalarParameter() {defaultValue=0.7, parameterName=Gradientoffset}) {constB=0.5})
  n20 = TextureSampleParameter2D() {
   parameterName=Mask
   samplerSource=SSM_Wrap_WorldGroupSettings
   samplerType=SAMPLERTYPE_Masks
   texture=/Game/Athena/Environments/Cattus/Demo/Textures/Cattus_Skull_Outside_M
  }
  n21 = LinearInterpolate(
   A=LinearInterpolate(
    A=VectorParameter() {parameterName=Color1, defaultValue=<rgba 0.6927 0.6927 0.6927 1>}.RGB
    B=VectorParameter() {parameterName=Color2, defaultValue=<rgba 1 1 1 1>}.RGB
    Alpha=@n20.G
   )
   B=VectorParameter() {parameterName=Color3, defaultValue=<rgba 0.2708 0.2412 0.1827 1>}.RGB
   Alpha=@n20.B
  )
  n22 = LinearInterpolate(
   A=@n21
   B=Multiply(A=@n21, B=ScalarParameter() {defaultValue=2, parameterName=FresnelBrightness})
   Alpha=Fresnel(ExponentIn=ScalarParameter() {defaultValue=1.7115, parameterName=FresnelBone}, Normal=VertexNormalWS()) {baseReflectFraction=0}
  )
  n23 = ScalarParameter() {defaultValue=3, parameterName=SizeBone_Tilingx}
  n24 = VertexNormalWS()
  n25 = MaterialFunctionCall!(
   Diffuse Texture Object=TextureObject() {texture=/Game/Athena/Environments/Cattus/Textures/T_Bone_D}
   Spec Texture Object=TextureObject() {samplerType=SAMPLERTYPE_Masks, texture=/Game/Athena/Environments/Cattus/Textures/T_Bone_M}
   TextureSize=@n23
   --------------------------
   Use seperate settings for Z projection
   Use seperate texture for Z projection
   Texture Size Z projection
   Z Plane Diffuse Texture Object
   Z Plane Spec Texture Object
   "--------------------------      "
   ProjectionTransitionContrast=ScalarParameter() {defaultValue=2, parameterName=BoneProjectionContrast}
   Use Texture Alpha Heightmap to Refine the Transistion
   HeightLerpTransitionContrast
   PickTheDiffuseHeightChannel
   TheDiffuseHeightChannel
   "--------------------------       "
   NormalTextureObject=TextureObject() {samplerType=SAMPLERTYPE_Normal, texture=/Game/Athena/Environments/Cattus/Textures/T_Bone_N}
   World Space Normals=StaticBool()
   UseUncorrectedNormalBlend
   "--------------------------  "
   Export Float 4 Diffuse
   Export Float 4 Spec
   "--------------------------     "
   WorldPosition=TransformPosition(None=WorldPosition().XYZ) {transformSourceType=TRANSFORMPOSSOURCE_World}
   World Space Normal=Normalize(VectorInput=Transform(None=@n24) {transformSourceType=TRANSFORMSOURCE_World, transformType=TRANSFORM_Local})
  ) {materialFunction=/Game/Packages/Fortress_SharedMaterials/Base_Material_Functions/WorldAlignedTextures_Complex_OptimzedNormal}
  n26 = MaterialFunctionCall(Float3=@n25.XYZ Spec) {materialFunction=&BreakOutFloat3Components}
  n27 = MakeMaterialAttributes!(
   BaseColor=Multiply(A=ShadingPathSwitch(Default=@n22, Deferred=@n22, Forward=@n21, Mobile=@n21), B=@n25.XYZ Diffuse Texture)
   Metallic
   Specular=@n26.R
   Roughness=Multiply(A=ScalarParameter() {defaultValue=3, parameterName=RoughnessBone}, B=@n26.B)
   Anisotropy
   EmissiveColor
   Opacity
   OpacityMask
   Normal=@n25.Normal XYZ Texture
   Tangent
   WorldPositionOffset
   SubsurfaceColor
   ClearCoat
   ClearCoatRoughness
   AmbientOcclusion
   Refraction
   CustomizedUVs_0
   CustomizedUVs_1
   CustomizedUVs_2
   CustomizedUVs_3
   CustomizedUVs_4
   CustomizedUVs_5
   CustomizedUVs_6
   CustomizedUVs_7
   PixelDepthOffset
   ShadingModel
   Displacement
  )
  n28 = StaticSwitchParameter(
   True=MakeMaterialAttributes(BaseColor=LinearInterpolate(A=@n1.RGB, B=@n7, Alpha=@n11), Metallic=@n12.G, Specular=@n12.G, Roughness=Add(A=@n14, B=@n12.B))
   False=StaticSwitchParameter(
    True=MaterialFunctionCall!(Material=@n27, Tint=@n15.RGB, TintMask=@n19) {materialFunction=/Engine/Functions/MaterialLayerFunctions/MatLayerBlend_Tint}.Blended Material
    False=@n27
   ) {parameterName=IsTeeth}
  ) {parameterName=IsSword}
  n29 = VertexColor()
  n30 = LinearInterpolate(
   A=ScalarParameter() {defaultValue=-1.2516, parameterName=GrassBlendDirtContrast}
   B=ScalarParameter() {defaultValue=0.85, parameterName=GrassBlendDirtContrast2}
   Alpha=@n29.G
  )
  n31 = LinearInterpolate(
   A=VectorParameter() {parameterName=ColorGravel, defaultValue=<rgba 0.5833 0.4972 0.4345 1>}.RGB
   B=VectorParameter() {parameterName=ColorGravel2, defaultValue=<rgba 0.2188 0.114 0.0376 1>}.RGB
   Alpha=@n30
  )
  n32 = ScalarParameter() {defaultValue=455.5468, parameterName=SizeForestFloor}
  n33 = MaterialFunctionCall(
   Diffuse Texture Object=TextureObject() {texture=/Game/Packages/DS_Fortnight/WhiteRockCustom_D}
   TextureSize=@n32
   Texture Size Z projection=@n32
   NormalTextureObject=TextureObject() {samplerType=SAMPLERTYPE_Normal, texture=/Game/Packages/DS_Fortnight/WhiteRockCustom_N}
   World Space Normals=StaticBool()
   Export Float 4 Diffuse=StaticBool() {value=true}
  ) {materialFunction=&WorldAlignedTextures_Complex_OptimzedNormal}
  n34 = ComponentMask(None=@n33.XYZ Diffuse Texture) {b=true, g=true, r=true}
  n35 = ScalarParameter() {defaultValue=0.4476, parameterName=Roughness_dirt}
  n36 = ScalarParameter() {defaultValue=1, parameterName=Roughness_dirt2}
  n37 = ComponentMask(None=@n33.XYZ Diffuse Texture) {r=true}
  n38 = ScalarParameter() {parameterName=Grave_Transition_low}
  n39 = ScalarParameter() {defaultValue=0.8389, parameterName=Gravel_Transition_high}
  n40 = Subtract(A=VertexColor().R) {constB=0.5}
  n41 = Saturate(None=@n30)
  n42 = ComponentMask(None=Add(A=VertexNormalWS(), B=Add(A=@n40, B=@n41))) {b=true}
  n43 = MaterialFunctionCall!(
   Base Material=@n28
   Top Material=MakeMaterialAttributes(
    BaseColor=Multiply(A=@n31, B=@n34)
    Specular=ScalarParameter() {defaultValue=0.0857, parameterName=SpecularDirt}
    Roughness=LinearInterpolate(A=@n35, B=@n36, Alpha=@n37)
    Normal=@n33.Normal XYZ Texture
   )
   Alpha=MaterialFunctionCall(
    Transition Phase=LinearInterpolate(A=@n38, B=@n39, Alpha=@n42)
    Height Texture=ComponentMask(None=@n33.XYZ Diffuse Texture) {a=true}
    Contrast=ScalarParameter() {defaultValue=-0.019, parameterName=Gravel_TranstionContrast}
   ) {materialFunction=&HeightLerp}.Alpha
  ) {materialFunction=/Engine/Functions/MaterialLayerFunctions/MatLayerBlend_Standard}
  n44 = MaterialFunctionCall(
   Diffuse Texture Object=TextureObject() {texture=/Game/Athena/Environments/Landscape/Textures/T_Athena_Grass_Farm_ColorMatched_D_2}
   TextureSize=ScalarParameter() {defaultValue=500, parameterName=GrassSize}
   World Space Normals=StaticBool()
   Export Float 4 Diffuse=StaticBool() {value=true}
  ) {materialFunction=&WorldAlignedTextures_Complex_OptimzedNormal}
  n45 = ComponentMask(None=@n44.XYZ Diffuse Texture) {b=true, g=true, r=true}
  n46 = VertexColor()
  n47 = ComponentMask(None=@n44.XYZ Diffuse Texture) {r=true}
  n48 = MakeMaterialAttributes(
   BaseColor=StaticSwitchParameter(
    True=MaterialFunctionCall!(In, Rock_SlopeMask, GrassMask, BaseColorTintOnlybyVertColor=@n45, VertexColorMask=@n46.B) {materialFunction=/Game/Athena/Environments/Cattus/MaterialFunctions/MF_CattusSpiral}.BaseColorOnlyOutput
    False=@n45
   ) {parameterName=UsePurpleGrowth}
   Specular=LinearInterpolate(
    A=ScalarParameter() {defaultValue=0.2381, parameterName=SpecularGrass1}
    B=ScalarParameter() {defaultValue=3, parameterName=SpecularGrass2}
    Alpha=@n47
   )
   Roughness=LinearInterpolate(
    A=ScalarParameter() {defaultValue=1, parameterName=Roughness_Grass}
    B=ScalarParameter() {defaultValue=-1, parameterName=Roughness_Grass2}
    Alpha=@n47
   )
  )
  n49 = MaterialFunctionCall(
   Diffuse Texture Object=TextureObject() {samplerType=SAMPLERTYPE_Masks, texture=/Game/Athena/Environments/Cattus/Textures/T_Bone_M2}
   TextureSize=ScalarParameter() {defaultValue=1529.9003, parameterName=VinesScale}
   ProjectionTransitionContrast=ScalarParameter() {defaultValue=-0.3048, parameterName=VinesProjectionContrast}
   World Space Normals=StaticBool()
   Export Float 4 Diffuse=StaticBool()
  ) {materialFunction=&WorldAlignedTextures_Complex_OptimzedNormal}
  n50 = ScalarParameter() {parameterName=ContrastVines}
  n51 = MaterialFunctionCall(
   Transition Phase=Multiply(
    A=MaterialFunctionCall(Transition Phase=@n29.G, Height Texture=@n49.XY Diffuse Texture, Contrast=@n50) {materialFunction=&HeightLerp}.Alpha
    B=ScalarParameter() {defaultValue=0.6667, parameterName=MaxGrass}
   )
   Height Texture=ComponentMask(None=@n44.XYZ Diffuse Texture) {a=true}
   Contrast=ScalarParameter() {defaultValue=1.0788, parameterName=TranstionContrastGrass}
  ) {materialFunction=&HeightLerp}
  n52 = GetMaterialAttributes(None=StaticSwitchParameter(
    True=MaterialFunctionCall(Base Material=@n43.Blended Material, Top Material=@n48, Alpha=@n51.Alpha) {materialFunction=&MatLayerBlend_Standard}.Blended Material
    False=@n28
   ) {defaultValue=true, parameterName=UseMossAndGravel}) {attributeGetTypes [#1]}
  n53 = MaterialFunctionCall(Float3=@n52.Normal) {materialFunction=&BreakOutFloat3Components}
  n54 = Multiply(A=@n53.R) {constB=-1}
  n55 = Constant3Vector() {constant=<rgba 0 0 1 0>}
  n56 = Multiply(A=Transform(None=@n55), B=Constant() {r=1})
  n57 = Multiply(A=@n53.B, B=ComponentMask(None=@n56) {g=true})
  n58 = ComponentMask(None=@n56) {r=true}
  n59 = Multiply(A=@n58, B=@n53.B)
  n60 = Multiply(A=@n53.R) {constB=-1}
  n61 = ScalarParameter() {parameterName=ContrastBlendNormals}
  n62 = LinearInterpolate(
   A=AppendVector(A=AppendVector(A=@n54, B=@n57), B=Multiply(A=@n53.G) {constB=-1})
   B=AppendVector(A=AppendVector(A=@n59, B=@n60), B=Multiply(A=@n53.G) {constB=-1})
   Alpha=MaterialFunctionCall(In=Abs(None=@n58), Contrast=@n61) {materialFunction=&CheapContrast}.Result
  )
  n63 = Constant3Vector() {constant=<rgba -1 1 1 0>}
  n64 = ComponentMask(None=@n56) {b=true}
  n65 = AppendVector(A=ComponentMask(None=Multiply(A=@n52.Normal, B=@n63)) {g=true, r=true}, B=Multiply(A=@n53.B, B=@n64))
  n66 = MaterialFunctionCall(In=Abs(None=@n64), Contrast=@n61) {materialFunction=&CheapContrast}
  n67 = Transform(None=Normalize(VectorInput=LinearInterpolate(A=@n62, B=@n65, Alpha=@n66.Result))) {transformSourceType=TRANSFORMSOURCE_World, transformType=TRANSFORM_Tangent}
  AppendVector(A=ComponentMask(None=@n67) {g=true, r=true})
  AppendVector(
   A=AppendVector(A=@n23, B=ScalarParameter() {defaultValue=2, parameterName=SizeBone_Tilingy})
   B=ScalarParameter() {defaultValue=2, parameterName=SizeBone_Tilingz}
  )
  Constant3Vector()
  MaterialFunctionCall() {materialFunction=&WorldAlignedNormal}
  n68 = Multiply(A=@n67, B=VectorParameter() {parameterName=NormalIntensity, defaultValue=<rgba 1 1 0 1>}.RGB)
  n69 = TextureSampleParameter2D() {parameterName=Normal, samplerType=SAMPLERTYPE_Normal, texture=/Game/Athena/Prototype/Blueprints/Doggus_Sword/Textures/T_Doggus_Sword_N}
  n70 = SetMaterialAttributes!(
   MaterialAttributes=@n52.MaterialAttributes
   Normal=MaterialProxyReplace(Realtime=Add(A=@n69.RGB, B=@n68), MaterialProxy=Constant3Vector() {constant=<rgba 0 0 1 0>})
  ) {attributeSetTypes [#1]}
  n71 = GetMaterialAttributes(None=@n70) {attributeGetTypes [#2]}
  n72 = TextureSampleParameter2D() {
   parameterName=Mask
   samplerSource=SSM_Wrap_WorldGroupSettings
   samplerType=SAMPLERTYPE_Masks
   texture=&Cattus_Skull_Outside_M
  }
  n73 = Saturate(None=Add(A=@n72.R, B=ScalarParameter() {parameterName=AOintensity}))
  n74 = StaticSwitchParameter(
   True=SetMaterialAttributes!(MaterialAttributes=@n71.MaterialAttributes, Base Color=Multiply(A=@n71.Base Color, B=@n73)) {attributeSetTypes [#2]}
   False=@n70
  ) {parameterName=UseAO}
  n75 = SetMaterialAttributes!(
   MaterialAttributes=GetMaterialAttributes(None=@n70).MaterialAttributes
   Specular=ScalarParameter() {defaultValue=0.1, parameterName=MobilSpec}
   Roughness=ScalarParameter() {defaultValue=0.8, parameterName=MobilRoughness}
  ) {attributeSetTypes [#3, #4]}
  n76 = ShadingPathSwitch(Default=@n74, Deferred=@n74, Forward=@n75, Mobile=@n75)
  n77 = MaterialFunctionCall!(
   WSDeformationCenter=VectorParameter() {bUseCustomPrimitiveData=true, group=z_code, parameterName=WSDeformationCenter, primitiveDataIndex=20}.RGB
   Code_DestructionAmount=ScalarParameter() {group=z_code, parameterName=Code_DestructionAmount}
   Code_UseUvForDamage
   Code_UseHitEffects=StaticBoolParameter() {defaultValue=true, group=z_code, parameterName=Code_UseHitEffects}
   Code_Emissive
   WSDamageRadius=ScalarParameter() {bUseCustomPrimitiveData=true, defaultValue=1, group=z_code, parameterName=WSDamageRadius, primitiveDataIndex=26}
   CrackDamageShader
   Diffuse
   ScalarModDamageCracks
   ScalarOffsetforDamageCracks
   ExposeLerpFunctionForInvertingTheHitDistanceCalc
   InvertDamageFalloff
   CrackWorldPosition
   WorldPosition
   DistanceCullFadeAnimationOffset=ScalarParameter() {bUseCustomPrimitiveData=true, group=z_code, parameterName=DistanceCullFadeAnimationOffset, primitiveDataIndex=12}
  ) {materialFunction=/Game/Packages/jl_tests/Fortnite_HitEffects}
  MaterialFunctionCall(MA In=StaticSwitchParameter(
    True=SetMaterialAttributes!(MaterialAttributes=@n76, World Position Offset=@n77.Hit And Shrink WPO) {attributeSetTypes [#5]}
    False=@n76
   ) {parameterName=IsTeeth}) {materialFunction=&MF_CreativeMoveTool}
  MaterialProxyReplace(Realtime=VertexInterpolator(VS=@n11).PS, MaterialProxy=@n11)
  Multiply(A=Constant() {r=-1}, B=@n55)
  Normalize(VectorInput=@n67)
  ScalarParameter() {defaultValue=8000, parameterName=SwordBlendDistance}
  ScalarParameter() {defaultValue=200, parameterName=SwordBlendOffset}
  Sign(None=MaterialFunctionCall() {materialFunction=/Engine/Functions/Engine_MaterialFunctions02/WorldPositionOffset/ObjectScale}.Scale X)
  Subtract(A=@n29.G) {constB=0.5}
  Transform(None=Constant3Vector() {constant=<rgba 0 0 1 0>}) {transformSourceType=TRANSFORMSOURCE_Local}
  VertexNormalWS()
 ]
}
